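# Idle Niagara effect for the TrackWaver Yes pickaxe, drawing orange glows and lights on skeletal mesh sockets.
# One emitter fires two bursts as separate spawn groups, which select socket index, color and camera offset and set VisibilityTag.
# Persistent particles follow their sockets, feeding mesh, sprite and light renderers.
# User.bisHeadHidden zeroes particle alpha and light radius when true.
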NiagaraSystem /BRCosmetics/Characters/Player/Female/Medium/Bodies/F_MED_TrackWaver_Yes/FX/NS_Pickaxe_TrackWaver_Yes_Idle
dependencies {
 usedDataInterfaces [&NiagaraDataInterfaceArrayInt32, &NiagaraDataInterfaceSkeletalMesh]
 usedDynamicInputs []
 usedModules [
  &self:SystemSpawnScript.SystemScriptSource.SystemScriptGraph.NiagaraNodeAssignment_0.SetVariables_#1
  &SystemState
 ]
}
emitters {Minimal {
  emitterUpdateScript [
   &EmitterState(Life Cycle Mode=::NewEnumerator1, Loop Behavior=::NewEnumerator1, Loop Duration=1, UseLoopDelay=false)
   &SpawnBurst_Instantaneous(Use Spawn Probability=false, UseLoopCountLimit=false)
   &SpawnBurst_Instantaneous(Spawn Count=3, Spawn Group=1, Use Spawn Probability=false, UseLoopCountLimit=false) {module=SpawnBurst_Instantaneous001}
  ]
  particleSpawnScript [
   /SetVariables(Particles.VisibilityTag=&SpawnGroup()) {isSetParameters=true}
   &InitializeParticle(
    Color=&MaskLinearColorBySpawnGroup(Masked Linear Color=<rgba 4 1.1454 0 3>, Passthrough Linear Color=<rgba 2 0.5727 0 0.25>)
    Mesh Scale Mode=::NewEnumerator3
    Sprite Size Mode=::NewEnumerator3
    Uniform Sprite Size=5
    UsePositionOffset=false
    Write Material Random=true
   )
   &InitialMeshOrientation(Mesh Axis To Orient=<0 0 0>, Mesh Orientation Mode=::NewEnumerator2, Orientation Vector=<0 0 0>, Use Rotation Vector=false) {enabled=false}
  ]
  particleUpdateScript [
   &ParticleState(Kill Particles When Lifetime Has Elapsed=false, Loop Particles Lifetime=true)
   &SkeletalMeshLocation(
    Bone / Socket Index=&MaskIntBySpawnGroup(
     Masked Int=0
     Passthrough Int=&SelectIntFromArray(
      Direct Array Index=&ReturnExecIndex()
      Int Selection Array=DataInterface() {propertyValues {IntData [0,1,2,3]}}
     )
     Spawn Group Mask=0
    )
    MaskSpawnGroup=false
    Mesh Sampling Type=::NewEnumerator5
    Skeletal Mesh=$System.Skeletal Mesh Reader
    Socket Sampling Mode=::NewEnumerator5
   )
   &CameraOffset(Camera Offset Amount=&MaskFloatBySpawnGroup(Masked Float=1, Passthrough Float=3))
   &FN_Light_Attributes(
    Color=$Particles.Color
    Create Color Attribute=true
    Create Exponent Attribute=true
    Create Position Attribute=true
    Create Radius Attribute=true
    Exponent=2
    Position=&AddVectorToPosition(Position=$Particles.Position, Vector=<0 3 5>)
    Radius=&MakeCustomFloatFromBool(Bool=$User.bisHeadHidden, True Float=0, False Float=16)
    Scale Color=4
   )
   &ScaleColor(
    Scale Alpha=&MakeCustomFloatFromBool(Bool=$User.bisHeadHidden, True Float=0, False Float=1)
    ScaleA=true
    ScaleRGB=true
   )
  ]
  rendererClasses [
   &NiagaraMeshRendererProperties
   &NiagaraSpriteRendererProperties
   &NiagaraLightRendererProperties
  ]
  renderers [
   {rendererClass=&NiagaraMeshRendererProperties, rendererIndex=0}
   {rendererClass=&NiagaraSpriteRendererProperties, rendererIndex=1}
   {rendererClass=&NiagaraLightRendererProperties, rendererIndex=2}
  ]
 }}
summary {
 systemName=NS_Pickaxe_TrackWaver_Yes_Idle
 emitters [{
   emitterName=Minimal
   rendererClasses [
    &NiagaraMeshRendererProperties
    &NiagaraSpriteRendererProperties
    &NiagaraLightRendererProperties
   ]
  }]
 userVariables [User.bisHeadHidden {type=NiagaraBool}]
}
