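# Physics asset for the TrackWaver female medium body, apparently for the skydive setup of its tail and skirt, judging by the name.
# Capsule bodies cover a nine-joint tail, both thighs, and left, middle and right skirt chains, with near-zero placeholder-sized capsules on some skirt bones.
# Swing and twist limits link each three-bone skirt chain to its master bone.
# Tail, thigh and dyn_Skirt side bodies have no constraints.
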PhysicsAsset /BRCosmetics/Characters/Player/Female/Medium/Bodies/F_MED_TrackWaver_Yes/Meshes/F_MED_TrackWaver_Yes_TailVis_Skydive_Physics
bodies {
 C_Tail_A_1_Jnt [{shapeType=Capsule, length=4.0112, radius=6.8687, shapeName=C_Tail_A_1_Jnt_capsule, center=<-0.0224 0 1.5654>, rotation=<rot 0 -90 93.213>}]
 C_Tail_A_2_Jnt [{shapeType=Capsule, length=5.4457, radius=6.8258, shapeName=C_Tail_A_2_Jnt_capsule, center=<-0.0627 0 0.916>, rotation=<rot 0 -90 93.4173>}]
 C_Tail_A_3_Jnt [{shapeType=Capsule, length=4.1373, radius=6.7426, shapeName=C_Tail_A_3_Jnt_capsule, center=<-0.0518 0 0.8194>, rotation=<rot 0 -90 91.1565>}]
 C_Tail_A_4_Jnt [{shapeType=Capsule, length=3.7943, radius=7.0856, shapeName=C_Tail_A_4_Jnt_capsule, center=<0 0 0.7488>, rotation=<rot 0 -90 90.0002>}]
 C_Tail_A_5_Jnt [{shapeType=Capsule, length=4.0805, radius=6.7994, shapeName=C_Tail_A_5_Jnt_capsule, center=<0 -0.0861 1.2689>, rotation=<rot 0 -90 90.0002>}]
 C_Tail_A_6_Jnt [{shapeType=Capsule, length=4.2658, radius=6.614, shapeName=C_Tail_A_6_Jnt_capsule, center=<0 0 0.9943>, rotation=<rot 0 -90 90.0002>}]
 C_Tail_A_7_Jnt [{shapeType=Capsule, length=4.0654, radius=6.8144, shapeName=C_Tail_A_7_Jnt_capsule, center=<0 0 1.0194>, rotation=<rot 0 -90 90.0002>}]
 C_Tail_A_8_Jnt [{shapeType=Capsule, length=3.6449, radius=6.7268, shapeName=C_Tail_A_8_Jnt_capsule, center=<0 0 1.2915>, rotation=<rot 0 -90 90.0002>}]
 C_Tail_A_9_Jnt [{shapeType=Capsule, length=4.9118, radius=7.968, shapeName=C_Tail_A_9_Jnt_capsule, center=<0 0 0.3203>, rotation=<rot 0 -90 90.0002>}]
 dyn_Skirt_l_01 [{
   shapeType=Capsule
   length=0.1
   radius=0.1
   shapeName=dyn_Skirt_l_01_capsule
   center=<8.9257 1.3941 -2.4549>
   rotation=<rot 7.963 -73.2561 -90>
  }]
 dyn_Skirt_r_01 [{
   shapeType=Capsule
   length=0.1
   radius=0.1
   shapeName=dyn_Skirt_r_01_capsule
   center=<-8.9257 -1.3941 2.4549>
   rotation=<rot 7.963 -73.2561 90>
  }]
 dyn_skirt_m_01 [{shapeType=Capsule, length=15, radius=5, shapeName=dyn_skirt_m_01_capsule, center=<8.9705 0 5.9816>, rotation=<rot 0.0004 0 90>}]
 dyn_skirt_m_02 [{shapeType=Capsule, length=4.9077, radius=5.548, shapeName=dyn_skirt_m_02_capsule, center=<6.4963 0 3.8671>, rotation=<rot 0.0001 0 90>}]
 dyn_skirt_m_03 [{shapeType=Capsule, length=10.966, radius=5, shapeName=dyn_skirt_m_03_capsule, center=<8.6378 0 2.5052>, rotation=<rot 0 -90 94.9995>}]
 dyn_skirt_m_master [{shapeType=Capsule, length=0, radius=0, shapeName=dyn_skirt_m_master_capsule}]
 dyn_tailSkirt_l_01 [{shapeType=Capsule, length=5.2871, radius=4.6674, shapeName=dyn_tailSkirt_l_01_capsule, center=<6.1543 0 3.3354>, rotation=<rot -90 0 0>}]
 dyn_tailSkirt_l_02 [{shapeType=Capsule, length=4.8265, radius=5.128, shapeName=dyn_tailSkirt_l_02_capsule, center=<6.1543 -1.5508 2.587>, rotation=<rot -90 0 0>}]
 dyn_tailSkirt_l_03 [{
   shapeType=Capsule
   length=6.2565
   radius=3.698
   shapeName=dyn_tailSkirt_l_03_capsule
   center=<5.7813 0.7803 2.3866>
   rotation=<rot -80.0001 0.0002 9.2977>
  }]
 dyn_tailSkirt_l_master [{shapeType=Capsule, length=2.2456, radius=0.505, shapeName=dyn_tailSkirt_l_master_capsule, rotation=<rot 0 -90 0>}]
 dyn_tailSkirt_r_01 [{shapeType=Capsule, length=5.2871, radius=4.6674, shapeName=dyn_tailSkirt_r_01_capsule, center=<-6.1543 0 -3.3354>, rotation=<rot -90 -180 0>}]
 dyn_tailSkirt_r_02 [{
   shapeType=Capsule
   length=4.8265
   radius=5.128
   shapeName=dyn_tailSkirt_r_02_capsule
   center=<-6.1543 1.6601 -2.587>
   rotation=<rot -90 -180 0>
  }]
 dyn_tailSkirt_r_03 [{
   shapeType=Capsule
   length=6.2565
   radius=3.698
   shapeName=dyn_tailSkirt_r_03_capsule
   center=<-5.7813 -0.7803 -2.3866>
   rotation=<rot -80.0001 0.0002 -170.7023>
  }]
 dyn_tailSkirt_r_master [{shapeType=Capsule, length=0.505, radius=0.505, shapeName=dyn_tailSkirt_r_master_capsule, rotation=<rot 0 -90 180>}]
 thigh_l [{
   shapeType=Capsule
   length=24.4843
   radius=9.2114
   shapeName=thigh_l_capsule
   center=<-18.9147 -0.3555 1.3302>
   rotation=<rot 2.6726 -85.7434 -88.2241>
  }]
 thigh_r [{
   shapeType=Capsule
   length=24.4843
   radius=9.2114
   shapeName=thigh_r_capsule
   center=<18.9147 0.3555 -1.3302>
   rotation=<rot 2.6726 -85.7434 91.7759>
  }]
}
constraints [
 {bone1Name=dyn_tailSkirt_l_02, bone2Name=dyn_tailSkirt_l_01, swing1LimitDegrees=20, swing2LimitDegrees=20, twistLimitDegrees=10}
 {bone1Name=dyn_tailSkirt_l_03, bone2Name=dyn_tailSkirt_l_02, swing1LimitDegrees=10, swing2LimitDegrees=10, twistLimitDegrees=5}
 {bone1Name=dyn_tailSkirt_l_01, bone2Name=dyn_tailSkirt_l_master, swing1LimitDegrees=15, swing2LimitDegrees=15, twistLimitDegrees=10}
 {bone1Name=dyn_tailSkirt_r_01, bone2Name=dyn_tailSkirt_r_master, swing1LimitDegrees=15, swing2LimitDegrees=15, twistLimitDegrees=10}
 {bone1Name=dyn_tailSkirt_r_02, bone2Name=dyn_tailSkirt_r_01, swing1LimitDegrees=20, swing2LimitDegrees=20, twistLimitDegrees=10}
 {bone1Name=dyn_tailSkirt_r_03, bone2Name=dyn_tailSkirt_r_02, swing1LimitDegrees=10, swing2LimitDegrees=10, twistLimitDegrees=5}
 {bone1Name=dyn_skirt_m_02, bone2Name=dyn_skirt_m_01, swing1LimitDegrees=20, swing2LimitDegrees=30, twistLimitDegrees=10}
 {bone1Name=dyn_skirt_m_03, bone2Name=dyn_skirt_m_02, swing1LimitDegrees=20, swing2LimitDegrees=30, twistLimitDegrees=10}
 {bone1Name=dyn_skirt_m_01, bone2Name=dyn_skirt_m_master, swing1LimitDegrees=20, swing2LimitDegrees=45, twistLimitDegrees=20}
]
