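# Post-process-style AnimBP for the TrackWaver female body cosmetic, adding secondary motion to skirt and tail.
# Copies the parent pose, layers in the tail mesh pose, then runs a Control Rig and back hip-gem constraints into a cached pose.
# That pose feeds RigidBody sims chosen by tail visibility and skydiving, then skirt AnimDynamics.
# The event graph updates velocity and emote state, and hides the tail when backbling is long.
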AnimBlueprint /BRCosmetics/Characters/Player/Female/Medium/Bodies/F_MED_TrackWaver_Yes/Meshes/F_MED_TrackWaver_Yes_AnimBP
target_skeleton=&/F_MED_TrackWaver_Yes_Skeleton
graphs [
 AnimGraph {nodes [
   n1 = Anim Dynamics - Bone: dyn_SkirtLo_l_02(Alpha=1, ExternalForce="0,0,0") {
    angularDampingOverride=0.9999
    angularSpringConstant=1500
    bAngularSpring=true
    bOverrideAngularDamping=true
    bUseGravityOverride=true
    boundBone {boneName=dyn_SkirtLo_l_02}
    gravityOverride {z=-980}
    physicsBodyDefinitions [{
      boundBone {boneName=dyn_SkirtLo_l_02}
      boxExtents=<10 10 10>
      constraintSetup {angularLimitsMax=<0 5 0>, angularLimitsMin=<0 -5 0>, angularTarget=<1 0 0>}
      localJointOffset=<0.1 0 0>
     }]
    simSpaceSettings {maxAngularAcceleration=30, maxAngularVelocity=8, simSpaceAngularAlpha=1}
   }
   Anim Dynamics - Bone: dyn_SkirtLo_r_02(ComponentPose=@n1, Alpha=1, ExternalForce="0,0,0") {
    angularDampingOverride=0.9999
    angularSpringConstant=1500
    bAngularSpring=true
    bOverrideAngularDamping=true
    bUseGravityOverride=true
    boundBone {boneName=dyn_SkirtLo_r_02}
    gravityOverride {z=-980}
    physicsBodyDefinitions [{
      boundBone {boneName=dyn_SkirtLo_r_02}
      boxExtents=<10 10 10>
      constraintSetup {angularLimitsMax=<0 5 0>, angularLimitsMin=<0 -5 0>, angularTarget=<1 0 0>}
      localJointOffset=<0.1 0 0>
     }]
    simSpaceSettings {maxAngularAcceleration=30, maxAngularVelocity=8, simSpaceAngularAlpha=1}
   }
   "Constraint - con_hipGemFront_l (PPPP)"!(
    ComponentPose
    Alpha=1
    ConstraintWeights_0=0.0313
    ConstraintWeights_1:float=0.3911
    ConstraintWeights_2:float=0.5572
    ConstraintWeights_3:float=0.0204
   ) -> (Pose) {
    boneToModify {boneName=con_hipGemFront_l}
    constraintSetup [
     {targetBone {boneName=pelvis}}
     {targetBone {boneName=thigh_twist_01_l}}
     {targetBone {boneName=thigh_l}}
     {targetBone {boneName=deform_groin_l}}
    ]
    constraintWeights [1,1,1,1]
   }
   "Constraint - con_hipGemFront_r (PPPP)"!(
    ComponentPose
    Alpha=1
    ConstraintWeights_0=0.0313
    ConstraintWeights_1:float=0.3911
    ConstraintWeights_2:float=0.5572
    ConstraintWeights_3:float=0.0204
   ) -> (Pose) {
    boneToModify {boneName=con_hipGemFront_r}
    constraintSetup [
     {targetBone {boneName=pelvis}}
     {targetBone {boneName=thigh_twist_01_r}}
     {targetBone {boneName=thigh_r}}
     {targetBone {boneName=deform_groin_r}}
    ]
    constraintWeights [1,1,1,1]
   }
   n2 = RigidBody!(
    ComponentPose=Local To Component(LocalPose=Use cached pose 'PreSim'())
    OverrideWorldGravity:Vector="0,0,-980"
    SimSpaceSettings:Sim Space Settings Structure=(WorldAlpha=1,VelocityScaleZ=1,DampingAlpha=1,MaxLinearVelocity=500,MaxAngularVelocity=7,MaxLinearAcceleration=1000,MaxAngularAcceleration=80,ExternalLinearDragV=(X=1,Y=1,Z=1),ExternalLinearVelocity=(X=0,Y=0,Z=0),ExternalAngularVelocity=(X=0,Y=0,Z=0))
   ) -> (Pose) {
    alphaInputType=Bool
    bClampLinearTranslationLimitToRefPose=true
    bOverrideWorldGravity=true
    overridePhysicsAsset=&/F_MED_TrackWaver_Yes_TailVis_Physics
    simulationSpace=BaseBoneSpace
    alphaBoolBlend {blendOption=Linear}
    baseBoneRef {boneName=pelvis}
    componentAppliedLinearAccClamp=<0 0 0>
    overrideWorldGravity {z=-980}
    simSpaceSettings {
     maxAngularAcceleration=80
     maxAngularVelocity=7
     maxLinearAcceleration=1000
     maxLinearVelocity=500
     worldAlpha=1
     externalLinearDragV=<1 1 1>
    }
   }
   n3 = "Constraint - dyn_tailSkirt_m_01 (RR)"!(ComponentPose=@n2, Alpha=1, ConstraintWeights_0=0.5, ConstraintWeights_1:float=0.5) -> (Pose) {
    boneToModify {boneName=dyn_tailSkirt_m_01}
    constraintSetup [{transformType=Rotation, targetBone {boneName=dyn_tailSkirt_l_01}}, {transformType=Rotation, targetBone {boneName=dyn_tailSkirt_r_01}}]
    constraintWeights [1,1]
   }
   n4 = RigidBody(
    ComponentPose=Local To Component(LocalPose=Use cached pose 'PreSim'())
    OverrideWorldGravity="0,0,-980"
    SimSpaceSettings=(WorldAlpha=1,VelocityScaleZ=1,DampingAlpha=1,MaxLinearVelocity=500,MaxAngularVelocity=7,MaxLinearAcceleration=1000,MaxAngularAcceleration=80,ExternalLinearDragV=(X=1,Y=1,Z=1),ExternalLinearVelocity=(X=0,Y=0,Z=0),ExternalAngularVelocity=(X=0,Y=0,Z=0))
   ) {
    alphaInputType=Bool
    bClampLinearTranslationLimitToRefPose=true
    bOverrideWorldGravity=true
    overridePhysicsAsset=&/F_MED_TrackWaver_Yes_TailVis_Skydive_Physics
    simulationSpace=BaseBoneSpace
    alphaBoolBlend {blendOption=Linear}
    baseBoneRef {boneName=pelvis}
    componentAppliedLinearAccClamp=<0 0 0>
    overrideWorldGravity {z=-980}
    simSpaceSettings {
     maxAngularAcceleration=80
     maxAngularVelocity=7
     maxLinearAcceleration=1000
     maxLinearVelocity=500
     worldAlpha=1
     externalLinearDragV=<1 1 1>
    }
   }
   n5 = Blend Poses by bool(BlendPose_0=Component To Local(ComponentPose=@n4), BlendPose_1=Component To Local(ComponentPose=@n3), bActiveValue=Get bIsSkyDiving())
   n6 = RigidBody(
    ComponentPose=Local To Component(LocalPose=Use cached pose 'PreSim'())
    OverrideWorldGravity="0,0,-980"
    SimSpaceSettings=(WorldAlpha=1,VelocityScaleZ=1,DampingAlpha=1,MaxLinearVelocity=500,MaxAngularVelocity=7,MaxLinearAcceleration=1000,MaxAngularAcceleration=80,ExternalLinearDragV=(X=1,Y=1,Z=1),ExternalLinearVelocity=(X=0,Y=0,Z=0),ExternalAngularVelocity=(X=0,Y=0,Z=0))
   ) {
    alphaInputType=Bool
    bClampLinearTranslationLimitToRefPose=true
    bOverrideWorldGravity=true
    overridePhysicsAsset=&/F_MED_TrackWaver_Yes_TailInvis_Physics
    simulationSpace=BaseBoneSpace
    alphaBoolBlend {blendOption=Linear}
    baseBoneRef {boneName=pelvis}
    componentAppliedLinearAccClamp=<0 0 0>
    overrideWorldGravity {z=-980}
    simSpaceSettings {
     maxAngularAcceleration=80
     maxAngularVelocity=7
     maxLinearAcceleration=1000
     maxLinearVelocity=500
     worldAlpha=1
     externalLinearDragV=<1 1 1>
    }
   }
   n7 = Component To Local(ComponentPose=@n6)
   n8 = Get isTailVisible() -> (isTailVisible:bool)
   n9 = Anim Dynamics - Bone: dyn_SkirtLo_l_02(
    ComponentPose=Local To Component(LocalPose=Blend Poses by bool(BlendPose_0=@n5, BlendPose_1=@n7, bActiveValue=@n8))
    Alpha=1
    ExternalForce="0,0,0"
   ) {
    angularDampingOverride=0.9999
    angularSpringConstant=100
    bAngularSpring=true
    bOverrideAngularDamping=true
    bUseGravityOverride=true
    boundBone {boneName=dyn_SkirtLo_l_02}
    componentAppliedLinearAccClamp=<1000 1000 1000>
    componentLinearAccScale=<0.5 0.5 0.5>
    componentLinearVelScale=<0.5 0.5 0.5>
    gravityOverride {z=-980}
    physicsBodyDefinitions [{
      boundBone {boneName=dyn_SkirtLo_l_02}
      boxExtents=<10 10 10>
      constraintSetup {angularLimitsMax=<0 5 5>, angularLimitsMin=<0 -20 -5>, angularTarget=<1 0 0>}
      localJointOffset=<0.4 0 0>
     }]
    simSpaceSettings {maxAngularAcceleration=30, maxAngularVelocity=8, simSpaceAngularAlpha=1}
   }
   n10 = Anim Dynamics - Bone: dyn_SkirtLo_r_02(ComponentPose=@n9, Alpha=1, ExternalForce="0,0,0") {
    angularDampingOverride=0.9999
    angularSpringConstant=100
    bAngularSpring=true
    bOverrideAngularDamping=true
    bUseGravityOverride=true
    boundBone {boneName=dyn_SkirtLo_r_02}
    componentAppliedLinearAccClamp=<800 800 800>
    componentLinearAccScale=<0.6 0.6 0.6>
    componentLinearVelScale=<0.6 0.6 0.6>
    gravityOverride {z=-980}
    physicsBodyDefinitions [{
      boundBone {boneName=dyn_SkirtLo_r_02}
      boxExtents=<10 10 10>
      constraintSetup {angularLimitsMax=<0 5 10>, angularLimitsMin=<0 -20 -10>, angularTarget=<1 0 0>}
      localJointOffset=<-0.4 0 0>
     }]
    simSpaceSettings {maxAngularAcceleration=30, maxAngularVelocity=8, simSpaceAngularAlpha=1}
   }
   n11 = "Transform (Modify) Bone - Bone: dyn_skirt_m_master"!(ComponentPose=@n10, bAlphaBoolEnabled:bool=True, Translation="0,0,0", Rotation="0,0,5", Scale="1,1,1") -> (Pose) {alphaInputType=Bool, rotationMode=BMM_Additive, translationMode=BMM_Additive, boneToModify {boneName=dyn_skirt_m_master}, rotation {roll=5}}
   Output Pose(Result=Component To Local(ComponentPose=@n11)) {layerGroup=None}
   n12 = Get Tail() -> (Tail:Skeletal Mesh Component Object Reference)
   n13 = Get isTailVisible()
   n14 = Control Rig!(
    Source=Layered blend per bone(
     BasePose=Copy Pose From Mesh() {bUseAttachedParent=true}
     BlendPoses_0=Copy Pose From Mesh(SourceMeshComponent=@n12) {bUseAttachedParent=true, bUseMeshPose=true}
     BlendWeights_0="To Float (Boolean)"(InBool=@n13)
    ) {bBlendRootMotionBasedOnRootBone=false, layerSetup [{branchFilters [{boneName=C_Root_Main_Root_Jnt}]}]}
    Alpha
    isTailVisible:bool=Get isTailVisible()
    Abs Velocity:double=Get abs velocity() -> (abs velocity:double)
    XY Velocity:double=Get XY Velocity() -> (XY Velocity:double)
    is DBNO Carried:bool=Get bIsDBNOCarried()
    isPlayingScrubDubEmote:bool=Get isPlayingScrubDubEmote() -> (isPlayingScrubDubEmote:bool)
    isPlayingEmote:bool=Get bIsPlayingEmote()
   ) -> (Pose) {controlRigAssetReference {blueprintRigClass=&/F_MED_TrackWaver_Yes_CtrlRig.F_MED_TrackWaver_Yes_CtrlRig_C}}
   n15 = "Constraint - con_hipGemBack_l (PPPPPP)"!(
    ComponentPose=Local To Component(LocalPose=@n14)
    Alpha=1
    ConstraintWeights_0=0.0025
    ConstraintWeights_1:float=0.387
    ConstraintWeights_2:float=0.5933
    ConstraintWeights_3:float=0
    ConstraintWeights_4:float=0.0001
    ConstraintWeights_5:float=0.0172
   ) -> (Pose) {
    boneToModify {boneName=con_hipGemBack_l}
    constraintSetup [
     {targetBone {boneName=pelvis}}
     {targetBone {boneName=thigh_twist_01_l}}
     {targetBone {boneName=thigh_l}}
     {targetBone {boneName=calf_l}}
     {targetBone {boneName=deform_knee_in_l}}
     {targetBone {boneName=deform_glute_l}}
    ]
    constraintWeights [1,1,1,1,1,1]
   }
   n16 = "Constraint - con_hipGemBack_r (PPPPPP)"!(
    ComponentPose=@n15
    Alpha=1
    ConstraintWeights_0=0.0025
    ConstraintWeights_1:float=0.387
    ConstraintWeights_2:float=0.5933
    ConstraintWeights_3:float=0
    ConstraintWeights_4:float=0.0001
    ConstraintWeights_5:float=0.0172
   ) -> (Pose) {
    boneToModify {boneName=con_hipGemBack_r}
    constraintSetup [
     {targetBone {boneName=pelvis}}
     {targetBone {boneName=thigh_twist_01_r}}
     {targetBone {boneName=thigh_r}}
     {targetBone {boneName=calf_r}}
     {targetBone {boneName=deform_knee_in_r}}
     {targetBone {boneName=deform_glute_r}}
    ]
    constraintWeights [1,1,1,1,1,1]
   }
   Save cached pose 'PreSim'(Pose=Component To Local(ComponentPose=@n16)) {cacheName=PreSim}
  ]}
 EventGraph {nodes [
   Event BlueprintUpdateAnimation()
   > n1 = Sequence!(execute) -> (then_0, then_1, then_2:Exec, then_3:Exec)
   > n2 = "Gravity Override State Function (Parameter Asset) (Nativized)"(
    CharacterPartAnimInstance=Self-Reference()
    GameTimeInSeconds=GetGameTimeInSeconds()
    Parameters=&/F_MED_TrackWaver_Yes_Parameters
   )
   > Set GravityOverride(GravityOverride=@n2)
   > Set RBAN Settings Out_Native(RBAN Settings Out_Native:Fort Rigid Body Anim Node Parameters Structure=@n2.OutRigidBodyAnimNodeParameters) -> (Output_Get:Fort Rigid Body Anim Node Parameters Structure)
   n3 = Get AnimInstance() -> (AnimInstance:Anim Instance Object Reference)
   Set isPlayingScrubDubEmote(
    execute=@n1.then_3
    isPlayingScrubDubEmote:bool="Equal (Object)"(
     A=GetCurrentActiveMontage!(self:Anim Instance Object Reference=@n3) -> (ReturnValue:Anim Montage Object Reference)
     B=/BRCosmetics/Animation/Game/MainPlayer/Emotes/Scrub_A_Dub/Emote_ScrubaDub_CMF_M
    )
   ) -> (Output_Get:bool)
   n4 = Get bIsPlayingEmote()
   n5 = Get bIsSliding()
   Set doCorrectPostPhysFE(
    execute=@n1.then_2
    doCorrectPostPhysFE:bool=AND Boolean!(A=Get bIsInFrontEnd(), B=NOT Boolean(A=@n4), C:bool=Get bRigidBodyEnabled(), D:bool=NOT Boolean(A=@n5)) -> (ReturnValue)
   ) -> (Output_Get:bool)
   n6 = Get PawnVelocity()
   Set abs velocity(execute=@n1.then_1, abs velocity:double=Vector Length(A=@n6)) -> (Output_Get:double)
   > Set XY Velocity(XY Velocity:double=Vector Length XY!(A:Vector=@n6) -> (ReturnValue:double)) -> (Output_Get:double)
   n7 = Event OnCharacterCustomizationLoaded()
   > n8 = Sequence!(execute) -> (then_0, then_1)
   > n9 = Cast To PlayerPawn_Athena(Object=TryGetPawnOwner())
   > Branch(Condition=CheckForTail!(self:Custom Character Part Anim Instance Object Reference) -> (bHasTail:bool))
   n10 = GetSkeletalMeshForPartType(self=@n9, PartType=MiscOrTail)
   > Set Tail(Tail:Skeletal Mesh Component Object Reference=@n10) -> (Output_Get:Skeletal Mesh Component Object Reference)
   > n11 = Branch(Condition=CheckForBackbling().bIsBackblingLong)
   > SetVisibility!(execute, self:Scene Component Object Reference=@n10, bNewVisibility:bool=false, bPropagateToChildren:bool=false) -> (then)
   > Set isTailVisible(isTailVisible:bool=false) -> (Output_Get:bool)
   SetVisibility(execute=@n11.else, self=@n10, bNewVisibility=true, bPropagateToChildren=false)
   > Set isTailVisible(isTailVisible=true)
   Set AnimInstance(execute=@n8.then_1, AnimInstance:Anim Instance Object Reference=GetAnimInstance(self=@n7.CustomBaseMesh)) -> (Output_Get:Anim Instance Object Reference)
  ]}
]
variables [
 abs velocity {type=number}
 AnimInstance {type=AnimInstance, default=null}
 doCorrectPostPhysFE {type=boolean}
 GravityOverride {type=Vector}
 isPlayingScrubDubEmote {type=boolean}
 isTailVisible {type=boolean}
 RBAN Settings Out_Native {type=FortRigidBodyAnimNodeParameters}
 Tail {type=SkeletalMeshComponent}
 XY Velocity {type=number}
]
