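# Anim Blueprint for the TrackWaver_Yes female cosmetic's face-accessory mesh, adding secondary physics to hair curls and tails.
# It copies the attached parent's pose, then chains single-bone AnimDynamics nodes with angular springs and limits in root-relative space, tuned stiffer for tails.
# Each update sets GravityOverride from a nativized function using the base TrackWaver parameter asset, though the dynamics nodes show fixed gravity instead of reading that variable.
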AnimBlueprint /BRCosmetics/Characters/Player/Female/Medium/Bodies/F_MED_TrackWaver_Yes/Meshes/Parts/F_MED_TrackWaver_Yes_FaceAcc_AnimBP
target_skeleton=&/F_MED_TrackWaver_Yes_FaceAcc_Skeleton
graphs [
 AnimGraph {nodes [
   n1 = Anim Dynamics - Bone: dyn_hairCurl_01!(
    ComponentPose=Local To Component(LocalPose=Copy Pose From Mesh() {bUseAttachedParent=true})
    Alpha=1
    GravityOverride:Vector="0,0,-980"
    ExternalForce="0,0,0"
   ) -> (Pose) {
    angularDampingOverride=1
    angularSpringConstant=500
    bAngularSpring=true
    bOverrideAngularDamping=true
    bUseGravityOverride=true
    simulationSpace=RootRelative
    worldSpaceMinimumScale=1
    boundBone {boneName=dyn_hairCurl_01}
    componentAppliedLinearAccClamp=<5000 5000 5000>
    componentLinearAccScale=<1 1 1>
    componentLinearVelScale=<1 1 1>
    gravityOverride {z=-980}
    physicsBodyDefinitions [{
      boundBone {boneName=dyn_hairCurl_01}
      boxExtents=<7 7 7>
      constraintSetup {angularLimitsMax=<1 5 15>, angularLimitsMin=<-1 -25 -15>, angularTarget=<1 0 0>}
      localJointOffset=<0.5 0 0>
     }]
    simSpaceSettings {maxAngularAcceleration=35, maxAngularVelocity=7, simSpaceAngularAlpha=1}
   }
   n2 = Anim Dynamics - Bone: dyn_hairCurl_02!(ComponentPose=@n1, Alpha=1, GravityOverride:Vector="0,0,-980", ExternalForce="0,0,0") -> (Pose) {
    angularDampingOverride=1
    angularSpringConstant=500
    bAngularSpring=true
    bOverrideAngularDamping=true
    bUseGravityOverride=true
    simulationSpace=RootRelative
    worldSpaceMinimumScale=1
    boundBone {boneName=dyn_hairCurl_02}
    componentAppliedLinearAccClamp=<5000 5000 5000>
    componentLinearAccScale=<1 1 1>
    componentLinearVelScale=<1 1 1>
    gravityOverride {z=-980}
    physicsBodyDefinitions [{
      boundBone {boneName=dyn_hairCurl_02}
      boxExtents=<7 7 7>
      constraintSetup {angularLimitsMax=<1 10 15>, angularLimitsMin=<-1 -30 -15>, angularTarget=<1 0 0>}
      localJointOffset=<0.5 0 0>
     }]
    simSpaceSettings {maxAngularAcceleration=35, maxAngularVelocity=7, simSpaceAngularAlpha=1}
   }
   n3 = Anim Dynamics - Bone: dyn_hairCurl_03!(ComponentPose=@n2, Alpha=1, GravityOverride:Vector="0,0,-980", ExternalForce="0,0,0") -> (Pose) {
    angularDampingOverride=1
    angularSpringConstant=500
    bAngularSpring=true
    bOverrideAngularDamping=true
    bUseGravityOverride=true
    simulationSpace=RootRelative
    worldSpaceMinimumScale=1
    boundBone {boneName=dyn_hairCurl_03}
    componentAppliedLinearAccClamp=<5000 5000 5000>
    componentLinearAccScale=<1 1 1>
    componentLinearVelScale=<1 1 1>
    gravityOverride {z=-980}
    physicsBodyDefinitions [{
      boundBone {boneName=dyn_hairCurl_03}
      boxExtents=<7 7 7>
      constraintSetup {angularLimitsMax=<1 5 15>, angularLimitsMin=<-1 -25 -15>, angularTarget=<1 0 0>}
      localJointOffset=<0.5 0 0>
     }]
    simSpaceSettings {maxAngularAcceleration=35, maxAngularVelocity=7, simSpaceAngularAlpha=1}
   }
   n4 = Anim Dynamics - Bone: dyn_hairTail_01!(ComponentPose=@n3, Alpha=1, GravityOverride:Vector="0,0,-980", ExternalForce="0,0,0") -> (Pose) {
    angularDampingOverride=1
    angularSpringConstant=1200
    bAngularSpring=true
    bOverrideAngularDamping=true
    bUseGravityOverride=true
    simulationSpace=RootRelative
    boundBone {boneName=dyn_hairTail_01}
    componentAppliedLinearAccClamp=<5000 5000 5000>
    componentLinearAccScale=<1 1 1>
    componentLinearVelScale=<1 1 1>
    gravityOverride {z=-980}
    physicsBodyDefinitions [{
      boundBone {boneName=dyn_hairTail_01}
      boxExtents=<7 7 7>
      constraintSetup {angularLimitsMax=<1 30 30>, angularLimitsMin=<-1 -30 -30>, angularTarget=<1 0 0>}
      localJointOffset=<0.3 0 0>
     }]
    simSpaceSettings {maxAngularAcceleration=35, maxAngularVelocity=10, simSpaceAngularAlpha=1}
   }
   n5 = Anim Dynamics - Bone: dyn_hairTail_a_01!(ComponentPose=@n4, Alpha=1, GravityOverride:Vector="0,0,-980", ExternalForce="0,0,0") -> (Pose) {
    angularDampingOverride=1
    angularSpringConstant=1200
    bAngularSpring=true
    bOverrideAngularDamping=true
    bUseGravityOverride=true
    simulationSpace=RootRelative
    boundBone {boneName=dyn_hairTail_a_01}
    componentAppliedLinearAccClamp=<5000 5000 5000>
    componentLinearAccScale=<1 1 1>
    componentLinearVelScale=<1 1 1>
    gravityOverride {z=-980}
    physicsBodyDefinitions [{
      boundBone {boneName=dyn_hairTail_a_01}
      boxExtents=<7 7 7>
      constraintSetup {angularLimitsMax=<1 30 30>, angularLimitsMin=<-1 -30 -30>, angularTarget=<1 0 0>}
      localJointOffset=<0.3 0 0>
     }]
    simSpaceSettings {maxAngularAcceleration=35, maxAngularVelocity=10, simSpaceAngularAlpha=1}
   }
   n6 = Anim Dynamics - Bone: dyn_hairTail_b_01!(ComponentPose=@n5, Alpha=1, GravityOverride:Vector="0,0,-980", ExternalForce="0,0,0") -> (Pose) {
    angularDampingOverride=1
    angularSpringConstant=1200
    bAngularSpring=true
    bOverrideAngularDamping=true
    bUseGravityOverride=true
    simulationSpace=RootRelative
    boundBone {boneName=dyn_hairTail_b_01}
    componentAppliedLinearAccClamp=<5000 5000 5000>
    componentLinearAccScale=<1 1 1>
    componentLinearVelScale=<1 1 1>
    gravityOverride {z=-980}
    physicsBodyDefinitions [{
      boundBone {boneName=dyn_hairTail_b_01}
      boxExtents=<7 7 7>
      constraintSetup {angularLimitsMax=<1 30 30>, angularLimitsMin=<-1 -30 -30>, angularTarget=<1 0 0>}
      localJointOffset=<0.3 0 0>
     }]
    simSpaceSettings {maxAngularAcceleration=35, maxAngularVelocity=10, simSpaceAngularAlpha=1}
   }
   n7 = Anim Dynamics - Bone: dyn_hairTail_c_01!(ComponentPose=@n6, Alpha=1, GravityOverride:Vector="0,0,-980", ExternalForce="0,0,0") -> (Pose) {
    angularDampingOverride=1
    angularSpringConstant=1200
    bAngularSpring=true
    bOverrideAngularDamping=true
    bUseGravityOverride=true
    simulationSpace=RootRelative
    boundBone {boneName=dyn_hairTail_c_01}
    componentAppliedLinearAccClamp=<5000 5000 5000>
    componentLinearAccScale=<1 1 1>
    componentLinearVelScale=<1 1 1>
    gravityOverride {z=-980}
    physicsBodyDefinitions [{
      boundBone {boneName=dyn_hairTail_c_01}
      boxExtents=<7 7 7>
      constraintSetup {angularLimitsMax=<1 30 30>, angularLimitsMin=<-1 -30 -30>, angularTarget=<1 0 0>}
      localJointOffset=<0.3 0 0>
     }]
    simSpaceSettings {maxAngularAcceleration=35, maxAngularVelocity=10, simSpaceAngularAlpha=1}
   }
   n8 = Anim Dynamics - Bone: dyn_hairTail_e_01!(ComponentPose=@n7, Alpha=1, GravityOverride:Vector="0,0,-980", ExternalForce="0,0,0") -> (Pose) {
    angularDampingOverride=1
    angularSpringConstant=1200
    bAngularSpring=true
    bOverrideAngularDamping=true
    bUseGravityOverride=true
    simulationSpace=RootRelative
    boundBone {boneName=dyn_hairTail_e_01}
    componentAppliedLinearAccClamp=<5000 5000 5000>
    componentLinearAccScale=<1 1 1>
    componentLinearVelScale=<1 1 1>
    gravityOverride {z=-980}
    physicsBodyDefinitions [{
      boundBone {boneName=dyn_hairTail_e_01}
      boxExtents=<7 7 7>
      constraintSetup {angularLimitsMax=<1 30 30>, angularLimitsMin=<-1 -30 -30>, angularTarget=<1 0 0>}
      localJointOffset=<0.3 0 0>
     }]
    simSpaceSettings {maxAngularAcceleration=35, maxAngularVelocity=10, simSpaceAngularAlpha=1}
   }
   n9 = Anim Dynamics - Bone: dyn_hairTail_f_01!(ComponentPose=@n8, Alpha=1, GravityOverride:Vector="0,0,-980", ExternalForce="0,0,0") -> (Pose) {
    angularDampingOverride=1
    angularSpringConstant=1200
    bAngularSpring=true
    bOverrideAngularDamping=true
    bUseGravityOverride=true
    simulationSpace=RootRelative
    boundBone {boneName=dyn_hairTail_f_01}
    componentAppliedLinearAccClamp=<5000 5000 5000>
    componentLinearAccScale=<1 1 1>
    componentLinearVelScale=<1 1 1>
    gravityOverride {z=-980}
    physicsBodyDefinitions [{
      boundBone {boneName=dyn_hairTail_f_01}
      boxExtents=<7 7 7>
      constraintSetup {angularLimitsMax=<1 30 30>, angularLimitsMin=<-1 -30 -30>, angularTarget=<1 0 0>}
      localJointOffset=<0.3 0 0>
     }]
    simSpaceSettings {maxAngularAcceleration=35, maxAngularVelocity=10, simSpaceAngularAlpha=1}
   }
   n10 = Anim Dynamics - Bone: dyn_hairTail_i_01!(ComponentPose=@n9, Alpha=1, GravityOverride:Vector="0,0,-980", ExternalForce="0,0,0") -> (Pose) {
    angularDampingOverride=1
    angularSpringConstant=1200
    bAngularSpring=true
    bOverrideAngularDamping=true
    bUseGravityOverride=true
    simulationSpace=RootRelative
    boundBone {boneName=dyn_hairTail_i_01}
    componentAppliedLinearAccClamp=<5000 5000 5000>
    componentLinearAccScale=<1 1 1>
    componentLinearVelScale=<1 1 1>
    gravityOverride {z=-980}
    physicsBodyDefinitions [{
      boundBone {boneName=dyn_hairTail_i_01}
      boxExtents=<7 7 7>
      constraintSetup {angularLimitsMax=<1 30 30>, angularLimitsMin=<-1 -30 -30>, angularTarget=<1 0 0>}
      localJointOffset=<0.3 0 0>
     }]
    simSpaceSettings {maxAngularAcceleration=35, maxAngularVelocity=10, simSpaceAngularAlpha=1}
   }
   n11 = Anim Dynamics - Bone: dyn_hairTail_d_01!(ComponentPose=@n10, Alpha=1, GravityOverride:Vector="0,0,-980", ExternalForce="0,0,0") -> (Pose) {
    angularDampingOverride=1
    angularSpringConstant=1200
    bAngularSpring=true
    bOverrideAngularDamping=true
    bUseGravityOverride=true
    simulationSpace=RootRelative
    boundBone {boneName=dyn_hairTail_d_01}
    componentAppliedLinearAccClamp=<5000 5000 5000>
    componentLinearAccScale=<1 1 1>
    componentLinearVelScale=<1 1 1>
    gravityOverride {z=-980}
    physicsBodyDefinitions [{
      boundBone {boneName=dyn_hairTail_d_01}
      boxExtents=<7 7 7>
      constraintSetup {angularLimitsMax=<1 30 30>, angularLimitsMin=<-1 -30 -30>, angularTarget=<1 0 0>}
      localJointOffset=<0.3 0 0>
     }]
    simSpaceSettings {maxAngularAcceleration=35, maxAngularVelocity=10, simSpaceAngularAlpha=1}
   }
   n12 = Anim Dynamics - Bone: dyn_hairTail_g_01!(ComponentPose=@n11, Alpha=1, GravityOverride:Vector="0,0,-980", ExternalForce="0,0,0") -> (Pose) {
    angularDampingOverride=1
    angularSpringConstant=1200
    bAngularSpring=true
    bOverrideAngularDamping=true
    bUseGravityOverride=true
    simulationSpace=RootRelative
    boundBone {boneName=dyn_hairTail_g_01}
    componentAppliedLinearAccClamp=<5000 5000 5000>
    componentLinearAccScale=<1 1 1>
    componentLinearVelScale=<1 1 1>
    gravityOverride {z=-980}
    physicsBodyDefinitions [{
      boundBone {boneName=dyn_hairTail_g_01}
      boxExtents=<7 7 7>
      constraintSetup {angularLimitsMax=<1 30 30>, angularLimitsMin=<-1 -30 -30>, angularTarget=<1 0 0>}
      localJointOffset=<0.3 0 0>
     }]
    simSpaceSettings {maxAngularAcceleration=35, maxAngularVelocity=10, simSpaceAngularAlpha=1}
   }
   n13 = Anim Dynamics - Bone: dyn_hairTail_h_01!(ComponentPose=@n12, Alpha=1, GravityOverride:Vector="0,0,-980", ExternalForce="0,0,0") -> (Pose) {
    angularDampingOverride=1
    angularSpringConstant=1200
    bAngularSpring=true
    bOverrideAngularDamping=true
    bUseGravityOverride=true
    simulationSpace=RootRelative
    boundBone {boneName=dyn_hairTail_h_01}
    componentAppliedLinearAccClamp=<5000 5000 5000>
    componentLinearAccScale=<1 1 1>
    componentLinearVelScale=<1 1 1>
    gravityOverride {z=-980}
    physicsBodyDefinitions [{
      boundBone {boneName=dyn_hairTail_h_01}
      boxExtents=<7 7 7>
      constraintSetup {angularLimitsMax=<1 30 30>, angularLimitsMin=<-1 -30 -30>, angularTarget=<1 0 0>}
      localJointOffset=<0.3 0 0>
     }]
    simSpaceSettings {maxAngularAcceleration=35, maxAngularVelocity=10, simSpaceAngularAlpha=1}
   }
   Output Pose(Result=Component To Local(ComponentPose=@n13)) {layerGroup=None}
  ]}
 EventGraph {nodes [
   Event BlueprintUpdateAnimation()
   > n1 = "Gravity Override State Function (Parameter Asset) (Nativized)"(
    CharacterPartAnimInstance=Self-Reference()
    GameTimeInSeconds=GetGameTimeInSeconds()
    Parameters=/BRCosmetics/Characters/Player/Female/Medium/Bodies/F_MED_TrackWaver/Meshes/F_MED_TrackWaver_Parameters
   )
   > Set GravityOverride(GravityOverride=@n1)
  ]}
]
variables [GravityOverride {type=Vector}]
